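# Niagara idle effect for the NoiseClueLurch glider: thruster flames, sparks, expanding sound rings and flickering lamp lights.
# All emitters attach via a shared socket reader, using spawn groups to pick thruster, ring or lamp sockets.
# User.ForwardDot gates fire and flame and blends spark and thruster intensity, with front-end overrides toning down some values.
# Rings fake chromatic fringing with red/blue scaled copies plus refraction, while lamps hue-shift over particle age.
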NiagaraSystem /BRCosmetics/Gadgets/Assets/VinderTech_GliderChute/Glider_NoiseClueLurch/FX/NS_Glider_NoiseClueLurch_Fireworks_Idle
dependencies {
 usedDataInterfaces [
  &NiagaraDataInterfaceCurve
  &NiagaraDataInterfaceVector2DCurve
  &NiagaraDataInterfaceVectorCurve
  /Script/Niagara.NiagaraDataInterfaceArrayQuat
  &NiagaraDataInterfaceArrayFloat
  &NiagaraDataInterfaceArrayInt32
  &NiagaraDataInterfaceArrayFloat3
  &NiagaraDataInterfaceSocketReader
 ]
 usedDynamicInputs [&MergeHelper_FloatSwitch, &SelectFloatFromArray]
 usedModules [
  &SystemState
  &self:SystemSpawnScript.SystemScriptSource.SystemScriptGraph.NiagaraNodeAssignment_0.SetVariables_#1
  &DoOnce
 ]
}
emitters {
 Fire {
  emitterUpdateScript [
   &EmitterState(Inactive Response=::NewEnumerator1, Life Cycle Mode=::NewEnumerator1, Loop Delay=0.1, Loop Duration=$Engine.DeltaTime, UseLoopDelay=false)
   &SpawnRate(
    Spawn Probability=&MakeCustomFloatFromBool(
     Bool=&SetBoolByFloatComparison(A=$User.ForwardDot, B=0.35, Comparison Type=::NewEnumerator2)
     True Float=1
     False Float=0
    )
    SpawnRate=20
    Use Spawn Probability=true
   )
  ]
  particleSpawnScript [
   &FN_SetSpawnGroup(Number of Spawn Groups=2)
   &InitializeParticle(
    Color=<rgba 2 0.5 0.01 1>
    Lifetime Max=0.25
    Lifetime Min=0.1
    Lifetime Mode=::NewEnumerator1
    Mesh Scale=<0.25 0.25 2>
    Mesh Scale Mode=::NewEnumerator4
    UsePositionOffset=true
    Write Material Random=true
   ) {module=InitializeParticle001}
   &InitialMeshOrientation(Orientation Vector=<-1 0 0>, Rotation=<0.25 0 0>, Use Orientation Vector=true, Use Rotation Vector=true) {module=InitialMeshOrientation001}
   &SocketLocation(
    SamplePosOffset=true
    Socket Index=&SpawnGroup()
    Socket Reader=$System.Socket Reader
    Socket Sampling Mode=::NewEnumerator3
    VelocitySpeedLimit=false
   )
  ]
  particleUpdateScript [
   &ParticleState()
   &SolveForcesAndVelocity(Clamp Velocity=false, Limit Acceleration=false)
   &ScaleMeshSize(Scale Factor=&Multiply_VectorByFloat(
     Vector=&MakeVector(
      X=&RandomRangeFloat(Minimum=0.75, Maximum=1)
      Y=&RandomRangeFloat(Minimum=0.75, Maximum=1)
      Z=&FloatFromCurve(
       FloatCurve=DataInterface() {propertyValues {Curve {defaultValue=1, keys [{tangentMode=RCTM_User, time=0, value=0.5}, {tangentMode=RCTM_User, time=1, value=1}]}}}
       Scale Curve=&RandomRangeFloat(Minimum=0.55, Maximum=0.75)
      )
     )
     Float=0.7
    ))
  ]
  rendererClasses [&NiagaraMeshRendererProperties]
  renderers [{rendererClass=&NiagaraMeshRendererProperties, rendererIndex=0}]
 }
 Flame {
  emitterUpdateScript [
   &EmitterState()
   &SpawnRate(SpawnRate=30, Use Spawn Probability=true)
  ]
  particleSpawnScript [
   &InitializeParticle(
    Lifetime Max=0.25
    Lifetime Min=0.15
    Lifetime Mode=::NewEnumerator1
    Mesh Scale Max=<0.4 0.3 0.3>
    Mesh Scale Min=<0.3 0.05 0.05>
    Mesh Scale Mode=::NewEnumerator2
    UsePositionOffset=false
    Write Material Random=true
   ) {module=InitializeParticle001}
   &FN_SetSpawnGroup(Number of Spawn Groups=2)
   /Game/Effects/Niagara/Modules/Orientation/FN_Fan(
    Cone Angle:NiagaraFloat=&UniformRangedFloat(Minimum=0, Maximum=0)
    Spin:NiagaraFloat=&UniformRangedFloat(Minimum=0, Maximum=6.2)
   )
   &InitialMeshOrientation(Mesh Orientation Mode=::NewEnumerator2, Orientation Vector=<0 1 0>, Use Rotation Vector=false) {module=InitialMeshOrientation001}
   &CameraOffset(Camera Offset Amount=2)
   &SocketLocation(
    SamplePosOffset=true
    Socket Index=&SpawnGroup()
    Socket Reader=$System.Socket Reader
    Socket Sampling Mode=::NewEnumerator3
    VelocitySpeedLimit=false
   )
  ]
  particleUpdateScript [
   &ParticleState()
   &Color(
    Color=<rgba 1 0.0833 0 1>
    Scale Color=&VectorFromCurve(VectorCurve=DataInterface() {propertyValues {
       XCurve {keys [<at -0.0031 48.8>,<at 1 1>]}
       YCurve {keys [<at -0.006 20.3>,<at 1 0.8>]}
       ZCurve {keys [<at -0.0011 1.0067>,<at 0.9989 0.28>]}
      }})
   )
   &ScaleMeshSize(Scale Factor=&VectorFromCurve(VectorCurve=DataInterface() {propertyValues {
       XCurve {keys [
         {interpMode=RCIM_Cubic, tangentMode=RCTM_User, time=0, value=0.55}
         {arriveTangent=0.0732, interpMode=RCIM_Cubic, leaveTangent=0.0732, tangentMode=RCTM_User, time=0.2, value=0.9}
         <at 1 0>
        ]}
       YCurve {keys [<at 0 0.6>,<at 0.2 0.95>,<at 1 0>]}
       ZCurve {keys [<at 0 0.55>,<at 0.2 0.9>,<at 1 0>]}
      }}))
   &KillParticles(Kill Particles=&SetBoolByFloatComparison(A=$User.ForwardDot, B=0.25, Comparison Type=::NewEnumerator2))
  ]
  rendererClasses [&NiagaraMeshRendererProperties]
  renderers [{rendererClass=&NiagaraMeshRendererProperties, rendererIndex=0}]
 }
 Lamp_Lights {
  emitterUpdateScript [
   &EmitterState(Inactive Response=::NewEnumerator1, Life Cycle Mode=::NewEnumerator1, Loop Delay=0.5, Loop Duration=1, UseLoopDelay=true)
   &SpawnBurst_Instantaneous(
    Spawn Count=2
    Spawn Probability=HlslExpression() {hlslExpression="Emitter.LoopCount == 0"}
    Use Spawn Probability=false
    UseLoopCountLimit=false
   )
  ]
  particleSpawnScript [
   &FN_SetSpawnGroup(Number of Spawn Groups=2)
   &InitialMeshOrientation(
    Mesh Axis To Orient=<0 1 0>
    Mesh Orientation Mode=::NewEnumerator2
    Orientation Coordinate Space=::Local
    Orientation Vector=<0 -1 0>
    Use Rotation Vector=true
   )
   &InitializeParticle(
    Color=<rgba 1 0 0 1>
    Lifetime=1.5
    Mesh Scale Mode=::NewEnumerator3
    Mesh Uniform Scale=0.9
    UsePositionOffset=false
    Write Material Random=true
   )
  ]
  particleUpdateScript [
   &ParticleState()
   &SocketLocation(
    SamplePosOffset=true
    Sampled Orientation Coordinate Space:ENiagaraCoordinateSpace=::Local
    Sampled Position Offset=<0 13 -2>
    Socket Index=&SelectIntFromArray(Direct Array Index=&SpawnGroup(), Int Selection Array=DataInterface() {propertyValues {IntData [8,9]}})
    Socket Reader=$System.Socket Reader
    Socket Sampling Mode=::NewEnumerator3
    VelocitySpeedLimit=false
   )
   /SetVariables(Particles.Flicker:NiagaraFloat=&FloatFromCurve(
     FloatCurve=DataInterface() {propertyValues {Curve {keys [
         {arriveTangent=0.1539, interpMode=RCIM_Cubic, leaveTangent=0.2923, tangentMode=RCTM_Break, time=0, value=0.75}
         {arriveTangent=1.0901, interpMode=RCIM_Cubic, leaveTangent=-0.1059, tangentMode=RCTM_Break, time=0.5037, value=1}
         {arriveTangent=-0.0561, interpMode=RCIM_Cubic, leaveTangent=-1.1371, tangentMode=RCTM_Break, time=1, value=0.75}
        ]}}}
     CurveIndex=$System.NormalizedLoopAge
    )) {isSetParameters=true, module=SetVariables_#2}
   &ScaleMeshSize(Scale Factor=&MakeVector(
     X=1.25
     Y=&FloatFromCurve(FloatCurve=DataInterface() {propertyValues {Curve {
         defaultValue=1
         keys [
          {interpMode=RCIM_Cubic, tangentMode=RCTM_User, time=0, value=0.75}
          {interpMode=RCIM_Cubic, tangentMode=RCTM_User, time=0.5, value=1}
          {tangentMode=RCTM_User, time=1, value=0.75}
         ]
        }}})
     Z=1.25
    ))
   &Color(
    Color=&Multiply_LinearColorByFloat(
     Linear Color=&HueShiftLinearColor(Hue=$Particles.NormalizedAge, Saturation=1, Value=1, Linear Color=$Particles.Initial.Color)
     Float=$Particles.Flicker
    )
    Scale Alpha=&FloatFromCurve(
     FloatCurve=DataInterface() {propertyValues {Curve {
        defaultValue=1
        keys [
         {interpMode=RCIM_Cubic, tangentMode=RCTM_User, time=0, value=0.75}
         {arriveTangent=0.1986, interpMode=RCIM_Cubic, leaveTangent=0.1986, tangentMode=RCTM_User, time=0.5, value=1}
         {tangentMode=RCTM_User, time=1, value=0.75}
        ]
       }}}
     CurveIndex=$System.NormalizedLoopAge
    )
   ) {module=Color001}
  ]
  rendererClasses [&NiagaraMeshRendererProperties]
  renderers [{rendererClass=&NiagaraMeshRendererProperties, rendererIndex=0}]
 }
 Light {
  emitterUpdateScript [
   &EmitterState(Life Cycle Mode=::NewEnumerator1, Loop Delay=0.5, Loop Duration=1, UseLoopDelay=true)
   &SpawnBurst_Instantaneous(Spawn Count=2, Use Spawn Probability=false, UseLoopCountLimit=false)
  ]
  particleSpawnScript [
   &FN_SetSpawnGroup(Number of Spawn Groups=2)
   &InitializeParticle(
    Color=<rgba 1 0 0 1>
    Lifetime=1.5
    Position Mode=::NewEnumerator0
    Sprite Size Mode=::NewEnumerator3
    Uniform Sprite Size=50
    Write Material Random=true
   )
   &CameraOffset(Camera Offset Amount=10)
   &SpriteFacingAndAlignment(
    Alignment Coordinate Space=::Local
    Facing Coordinate Space=::Local
    Sprite Alignment=<0 0 1>
    Sprite Facing=<0 1 0>
    WriteAlignment=true
    WriteFacing=true
   )
   &InitialMeshOrientation(
    Mesh Orientation Mode=::NewEnumerator2
    Orientation Coordinate Space=::Local
    Orientation Vector=<0 1 0>
    Rotation=&SelectVectorFromArray(
     Direct Array Index=&SpawnGroup()
     Vector Selection Array=DataInterface() {propertyValues {FloatData [<0 -0.085 0>, <0 -0.085 0>]}}
    )
    Use Rotation Vector=false
   ) {module=InitialMeshOrientation001}
  ]
  particleUpdateScript [
   &ParticleState()
   /SetVariables(Particles.Flicker:NiagaraFloat=&FloatFromCurve(
     FloatCurve=DataInterface() {propertyValues {Curve {keys [
         {arriveTangent=0.1539, interpMode=RCIM_Cubic, leaveTangent=0.2923, tangentMode=RCTM_Break, time=0, value=0.75}
         {arriveTangent=1.0901, interpMode=RCIM_Cubic, leaveTangent=-0.1059, tangentMode=RCTM_Break, time=0.5037, value=1}
         {arriveTangent=-0.0561, interpMode=RCIM_Cubic, leaveTangent=-1.1371, tangentMode=RCTM_Break, time=1, value=0.75}
        ]}}}
     CurveIndex=$System.NormalizedLoopAge
    )) {isSetParameters=true, module=SetVariables_#2}
   &SocketLocation(
    SamplePosOffset=true
    Sampled Orientation Coordinate Space:ENiagaraCoordinateSpace=::Local
    Sampled Position Offset=&SelectVectorFromArray(
     Direct Array Index=&SpawnGroup()
     Vector Selection Array=DataInterface() {propertyValues {FloatData [<0 75 0>, <0 65 0>]}}
    )
    Socket Index=&SelectIntFromArray(Direct Array Index=&SpawnGroup(), Int Selection Array=DataInterface() {propertyValues {IntData [8,9]}})
    Socket Reader=$System.Socket Reader
    Socket Sampling Mode=::NewEnumerator3
    VelocitySpeedLimit=false
   )
   &FN_Light_Attributes(
    Color=&Multiply_LinearColorByFloat(
     Linear Color=&HueShiftLinearColor(Hue=$Particles.NormalizedAge, Saturation=1, Value=1, Linear Color=$Particles.Initial.Color)
     Float=$Particles.Flicker
    )
    Create Color Attribute=true
    Create Exponent Attribute=true
    Create Radius Attribute=true
    Exponent=5
    Radius=35
    Scale Color=35
   )
   &FN_Light_Attributes(
    Color=&Multiply_LinearColorByFloat(
     Linear Color=&HueShiftLinearColor(Hue=$Particles.NormalizedAge, Saturation=1, Value=1, Linear Color=$Particles.Initial.Color)
     Float=$Particles.Flicker
    )
    Create Color Attribute=true
    Create Position Attribute=true
    Create Radius Attribute=true
    Enable Particle Light Enable=true
    Light Enabled=true
    Position=&SelectVectorFromArray(
     Direct Array Index=&SpawnGroup()
     Vector Selection Array=DataInterface() {propertyValues {FloatData [<-40 60 18>, <40 60 18>]}}
    )
    Radius=220
    Scale Color=3
   ) {module=FN_Light_Attributes001}
   &ScaleColor(
    Scale Alpha=&FloatFromCurve(
     FloatCurve=DataInterface() {propertyValues {Curve {
        defaultValue=1
        keys [
         {arriveTangent=3.1934, interpMode=RCIM_Cubic, leaveTangent=3.1934, tangentMode=RCTM_User, time=0, value=0.25}
         {arriveTangent=0.078, interpMode=RCIM_Cubic, leaveTangent=0.078, tangentMode=RCTM_User, time=0.75, value=1}
         {arriveTangent=-2.6622, tangentMode=RCTM_User, time=1, value=0.25}
        ]
       }}}
     CurveIndex=$System.NormalizedLoopAge
    )
    ScaleA=true
    ScaleRGB=false
   )
   &Color(
    Color=&Multiply_LinearColorByFloat(
     Linear Color=&HueShiftLinearColor(Hue=$Particles.NormalizedAge, Saturation=1, Value=1, Linear Color=$Particles.Initial.Color)
     Float=$Particles.Flicker
    )
    Scale Alpha=&FloatFromCurve(
     FloatCurve=DataInterface() {propertyValues {Curve {
        defaultValue=1
        keys [
         {arriveTangent=3.1934, interpMode=RCIM_Cubic, leaveTangent=3.1934, tangentMode=RCTM_User, time=0, value=0.25}
         {arriveTangent=0.078, interpMode=RCIM_Cubic, leaveTangent=0.078, tangentMode=RCTM_User, time=0.75, value=1}
         {arriveTangent=-2.6622, tangentMode=RCTM_User, time=1, value=0.25}
        ]
       }}}
     CurveIndex=$System.NormalizedLoopAge
     Scale Curve=0.5
    )
   ) {module=Color002}
  ]
  rendererClasses [
   &NiagaraLightRendererProperties
   &NiagaraComponentRendererProperties
   &NiagaraSpriteRendererProperties
  ]
  renderers [
   {rendererClass=&NiagaraLightRendererProperties, rendererIndex=0}
   {rendererClass=&NiagaraComponentRendererProperties, rendererIndex=1}
   {rendererClass=&NiagaraSpriteRendererProperties, rendererIndex=2}
  ]
 }
 Sound_Rings {
  emitterUpdateScript [
   &EmitterState(Inactive Response=::NewEnumerator1, Life Cycle Mode=::NewEnumerator1, Loop Delay=0.35, Loop Duration=1, UseLoopDelay=false)
   &SpawnRate(
    SpawnRate=&FloatFromCurve(
     FloatCurve=DataInterface() {propertyValues {Curve {
        defaultValue=1
        keys [
         {interpMode=RCIM_Cubic, tangentMode=RCTM_User, time=0, value=0.2}
         {interpMode=RCIM_Cubic, tangentMode=RCTM_User, time=0.5, value=1}
         {tangentMode=RCTM_User, time=1, value=0.2}
        ]
       }}}
     CurveIndex=$Emitter.NormalizedLoopAge
     Scale Curve=30
    )
    Use Spawn Probability=false
   )
  ]
  particleSpawnScript [
   &FN_SetSpawnGroup(Number of Spawn Groups=6)
   &InitializeParticle(
    Color=<rgba 0.4 1 0.4 1>
    Lifetime=0.35
    Mesh Scale Mode=::NewEnumerator3
    Mesh Uniform Scale=0.1
    UsePositionOffset=false
    Write Material Random=true
   )
   &InitialMeshOrientation(Mesh Axis To Orient=<0 1 0>, Mesh Orientation Mode=::NewEnumerator2, Orientation Vector=<0 0 1>, Use Rotation Vector=false)
   &AddVelocity(
    Additional Quaternion Rotation=false
    Rotation Mode=::NewEnumerator3
    Rotation Quaternion:Quat4f=/Niagara/DynamicInputs/Arrays/SelectQuaternionFromArray(
     Array Sampling Mode:ENiagaraArraySamplingMode=::NewEnumerator1
     Direct Array Index:NiagaraInt32=&SpawnGroup()
     Interpolation:NiagaraFloat=?
     Direct Array Mode:ENiagara_WrapClamp=::NewEnumerator0
     Evaluation Type:ENiagaraRandomnessEvaluation=::NewEnumerator0
     Normalized Interpolation:NiagaraBool=true
     Recalculate Random Each Loop:NiagaraBool=false
     Randomness Mode:ENiagaraRandomnessMode=::NewEnumerator0
     Random Seed:NiagaraInt32=?
     Fixed Random Seed:NiagaraBool=false
     Quaternion Selection Array:NiagaraDataInterfaceArrayQuat=DataInterface() {propertyValues {
       GpuSyncMode=SyncCpuToGpu
       MaxElements=0
       QuatData [<0 0 0.7071 0.7071>, <0 0 0.3827 0.9239>, <0 0 0 1>, <0 0 0 1>, <0 0 -0.3827 0.9239>, <0 0 -0.7071 0.7071>]
      }}
     Enable Spherical Linear Interpolation:NiagaraBool=?
    )
    Velocity=<0 -1 0>
    Velocity Speed Scale=&RandomRangeFloat(Minimum=25, Maximum=35)
   )
   &SocketLocation(
    SamplePosOffset=false
    Sampled Orientation Coordinate Space:ENiagaraCoordinateSpace=::Local
    Sampled Position Coordinate Space:ENiagaraCoordinateSpace=::Local
    Sampled Velocity Coordinate Space:ENiagaraCoordinateSpace=::Local
    Socket Index=&SelectIntFromArray(Direct Array Index=&SpawnGroup(), Int Selection Array=DataInterface() {propertyValues {IntData [2,3,4,5,6,7]}})
    Socket Reader=$System.Socket Reader
    Socket Sampling Mode=::NewEnumerator3
    VelocitySpeedLimit=false
   )
   &ApplyInitialForces()
   /SetVariables() {isSetParameters=true}
  ]
  particleUpdateScript [
   &ParticleState()
   &ScaleMeshSize(Scale Factor=&VectorFromFloat(Value=&FloatFromCurve(FloatCurve=DataInterface() {propertyValues {Curve {
         defaultValue=1
         keys [{interpMode=RCIM_Cubic, tangentMode=RCTM_Break, time=0, value=0.85}, {interpMode=RCIM_Cubic, tangentMode=RCTM_Break, time=1, value=3}]
        }}}))) {module=ScaleMeshSize001}
   &ScaleColor(
    Scale Alpha=&FloatFromCurve(
     FloatCurve=DataInterface() {propertyValues {Curve {keys [
         {arriveTangent=51.3864, interpMode=RCIM_Cubic, leaveTangent=51.3864, tangentMode=RCTM_User, time=0, value=1}
         {arriveTangent=-0.4511, interpMode=RCIM_Cubic, leaveTangent=-0.4511, tangentMode=RCTM_User, time=0.0538, value=3.5}
         {arriveTangent=-4.1942, interpMode=RCIM_Cubic, leaveTangent=-4.1942, tangentMode=RCTM_User, time=0.9, value=0}
        ]}}}
     Scale Curve=0.05
    )
    ScaleA=true
    ScaleRGB=true
   )
   /SetVariables(
    Particles.ColorBlueSmall:LinearColor=&MakeLinearColor(
     R:NiagaraFloat=0.2
     G:NiagaraFloat=0.2
     B:NiagaraFloat=1
     A:NiagaraFloat=&MakeFloatFromLinearColor(LinearColor=$Particles.Color, Channel:ENiagaraFloatFromLinearColorOptions=::NewEnumerator3)
    )
    Particles.ColorRedSmall:LinearColor=&MakeLinearColor(
     R:NiagaraFloat=1
     G:NiagaraFloat=0.2
     B:NiagaraFloat=0.2
     A:NiagaraFloat=&MakeFloatFromLinearColor(LinearColor=$Particles.Color, Channel:ENiagaraFloatFromLinearColorOptions=::NewEnumerator3)
    )
    Particles.RefractionScaleSize:Vector3f=&MakeVector(
     X=1
     Y=1
     Z=&FloatFromCurve(
      FloatCurve=DataInterface() {propertyValues {Curve {
         defaultValue=1
         keys [
          {interpMode=RCIM_Cubic, tangentMode=RCTM_User, time=0, value=0.35}
          {interpMode=RCIM_Cubic, tangentMode=RCTM_User, time=0.5, value=1}
          {tangentMode=RCTM_User, time=1, value=0.35}
         ]
        }}}
      Scale Curve=2
     )
    )
    Particles.ScaleBlueSmall:Vector3f=&Multiply_VectorByFloat(Vector=$Particles.Scale, Float=&Subtract_Float(A=1, B=0.05))
    Particles.ScaleRedSmall:Vector3f=&Multiply_VectorByFloat(Vector=$Particles.Scale, Float=&Add_Float(A=1, B=0.05))
   ) {isSetParameters=true}
   /SetVariables(Particles.DynamicMaterialParameter0_Refraction:Vector4f=&MakeVector4(
     X=&FloatFromCurve(
      FloatCurve=DataInterface() {propertyValues {Curve {keys [
          {arriveTangent=-0.0008, interpMode=RCIM_Cubic, leaveTangent=-0.0008, tangentMode=RCTM_User, time=0, value=1.005}
          {arriveTangent=-0.0099, interpMode=RCIM_Cubic, leaveTangent=-0.0099, tangentMode=RCTM_User, time=1, value=1}
         ]}}}
      Scale Curve=&FN_FrontEndFloatModifier(Gameplay=1, Front End=1.005)
     )
     Y=1
     Z=1
     W=0
    )) {isSetParameters=true}
   &SolveForcesAndVelocity(Clamp Velocity=false, Limit Acceleration=false)
  ]
  rendererClasses [&NiagaraMeshRendererProperties]
  renderers [
   {rendererClass=&NiagaraMeshRendererProperties, rendererIndex=0}
   {rendererClass=&NiagaraMeshRendererProperties, rendererIndex=1}
   {rendererClass=&NiagaraMeshRendererProperties, rendererIndex=2}
   {rendererClass=&NiagaraMeshRendererProperties, rendererIndex=3}
  ]
 }
 Sparks {
  emitterUpdateScript [
   &EmitterState(Life Cycle Mode=::NewEnumerator1, Loop Duration=$Engine.DeltaTime, UseLoopDelay=false)
   &SpawnBurst_Instantaneous(
    Spawn Count=&RandomRangeInt(Minimum=25, Maximum=25)
    Spawn Probability=&MakeCustomFloatFromBool(Bool=$System.DoOnce.Execute, True Float=1, False Float=0)
    Use Spawn Probability=true
    UseLoopCountLimit=false
   )
   &SpawnRate(
    SpawnRate=&FloatFromCurve(
     FloatCurve=DataInterface() {propertyValues {Curve {
        defaultValue=1
        keys [
         {interpMode=RCIM_Cubic, tangentMode=RCTM_User, time=0, value=0.25}
         {interpMode=RCIM_Cubic, tangentMode=RCTM_User, time=0.5, value=1}
         {tangentMode=RCTM_User, time=1, value=0.25}
        ]
       }}}
     CurveIndex=$Emitter.NormalizedLoopAge
     Scale Curve=150
    )
    Use Spawn Probability=false
   ) {module=SpawnRate001}
  ]
  particleSpawnScript [
   &FN_SetSpawnGroup(Number of Spawn Groups=2)
   &InitializeParticle(
    Lifetime Max=0.35
    Lifetime Min=0.25
    Lifetime Mode=::NewEnumerator1
    Position Mode=::NewEnumerator0
    Sprite Size Max=<4 24>
    Sprite Size Min=<2 16>
    Sprite Size Mode=::NewEnumerator2
    Write Material Random=true
   )
   &SocketLocation(
    SamplePosOffset=true
    Sampled Position Offset=<0 5 0>
    Socket Index=&SpawnGroup()
    Socket Reader=$System.Socket Reader
    Socket Sampling Mode=::NewEnumerator3
    VelocitySpeedLimit=false
   )
   &CameraOffset(Camera Offset Amount=5)
   &AddVelocity(
    Cone Angle=&Lerp_Float(A=15, B=20, Alpha=$User.ForwardDot)
    Cone Axis=<0 1 0>
    Inner Cone Angle:NiagaraFloat=&Lerp_Float(A=25, B=5, Alpha=$User.ForwardDot)
    Module.UseConeFalloff=false
    Module.UseVelDistribution=false
    Override Seed=false
    Velocity Mode=::NewEnumerator2
    Velocity Speed=&Lerp_Float(
     A=&Multiply_Float(A=250, B=&RandomRangeFloat(Minimum=0.3, Maximum=1))
     B=&Multiply_Float(A=750, B=&RandomRangeFloat(Minimum=0.3, Maximum=1))
     Alpha=$User.ForwardDot
    )
   )
   /SetVariables(Particles.RandomLerp:NiagaraFloat=&MakeCustomFloatFromBool(Bool=&RandomBool(), True Float=0, False Float=1)) {isSetParameters=true}
  ]
  particleUpdateScript [
   &ParticleState()
   &ScaleColor(
    Scale Alpha=&MakeCustomFloatFromBool(Bool=$User.bIsFrontEndPreview, True Float=0.1, False Float=1)
    Scale RGB=&MakeVectorFromLinearColorRGB(Color=&LerpLinearColors(StartColor=<rgba 5 3.2081 2.12 1>, EndColor=<rgba 5 3.0994 1.1022 1>, LerpFactor=$Particles.RandomLerp))
    ScaleA=true
    ScaleRGB=true
   ) {module=ScaleColor001}
   &ScaleSpriteSize(
    Scale Factor=&Lerp_Vector2(
     A=&Vector2DFromCurve(
      Vector2Curve=DataInterface() {propertyValues {
        XCurve {defaultValue=0, keys [{interpMode=RCIM_Cubic, tangentMode=RCTM_Break, time=0, value=1}, {tangentMode=RCTM_User, time=1, value=0}]}
        YCurve {defaultValue=1, keys [{interpMode=RCIM_Cubic, tangentMode=RCTM_Break, time=0, value=1}, {tangentMode=RCTM_User, time=1, value=0}]}
       }}
      ScaleCurve=&RandomRangeVector2D(Minimum=<0.35 0.35>, Maximum=<0.55 0.55>)
     )
     B=&Vector2DFromCurve(
      Vector2Curve=DataInterface() {propertyValues {
        XCurve {defaultValue=1, keys [{interpMode=RCIM_Cubic, tangentMode=RCTM_Break, time=0, value=0}, {tangentMode=RCTM_User, time=1, value=1}]}
        YCurve {defaultValue=0, keys [{interpMode=RCIM_Cubic, tangentMode=RCTM_Break, time=0, value=1}, {tangentMode=RCTM_User, time=1, value=0}]}
       }}
      ScaleCurve=&RandomRangeVector2D(Minimum=<0.5 1>, Maximum=<1 2>)
     )
     Alpha=$User.ForwardDot
    )
    Scale Sprite Size Mode=::NewEnumerator2
   ) {module=ScaleSpriteSize001}
   &CurlNoiseForce(
    Bias Noise=false
    Noise Frequency=300
    Noise Strength=&Lerp_Float(A=1500, B=2500, Alpha=$User.ForwardDot)
    Pan Noise=true
    Pan Noise Field=<1 0 0>
    Random Seed=10
   )
   &Drag(Drag=&Lerp_Float(A=2, B=4, Alpha=$User.ForwardDot), Use Linear Drag=true, Use Rotational Drag=true)
   &SolveForcesAndVelocity(Clamp Velocity=false, Limit Acceleration=false)
  ]
  rendererClasses [&NiagaraSpriteRendererProperties]
  renderers [{rendererClass=&NiagaraSpriteRendererProperties}]
 }
 Thursters {
  emitterUpdateScript [
   &EmitterState(
    Inactive Response=::NewEnumerator1
    Life Cycle Mode=::NewEnumerator1
    Loop Behavior=::NewEnumerator1
    Loop Duration Mode=::NewEnumerator1
    UseLoopDelay=true
   )
   &SpawnBurst_Instantaneous(
    Spawn Count=2
    Spawn Probability=HlslExpression() {hlslExpression="Emitter.LoopCount == 0"}
    Use Spawn Probability=false
    UseLoopCountLimit=true
   )
  ]
  particleSpawnScript [
   &FN_SetSpawnGroup(Number of Spawn Groups=2)
   &InitializeParticle(Color Mode=::NewEnumerator0, UsePositionOffset=false, Write Material Random=true)
  ]
  particleUpdateScript [
   &ParticleState(Kill Particles When Lifetime Has Elapsed=false, Let Infinitely Lived Particles Die When Emitter Deactivates=true)
   /SetVariables(Particles.Jitter:NiagaraFloat=&Waveform(
     Global Input Position=&RandomRangeFloat(Minimum=0, Maximum=1, Evaluation Type=::NewEnumerator1)
     "Waveform [1]"=::NewEnumerator12
     "[1] Frequency"=20
     "[1] Amplitude Min/Max"=<0.7 1.1>
    )) {isSetParameters=true}
   &SocketLocation(
    SamplePosOffset=true
    Sampled Position Offset=<0 3 0>
    Socket Index=&MergeHelper_IntSwitch(0=0, 1=1, 2=?, 3=?, 4=?, 5=?, Index=$Particles.SpawnGroup, Num=::NewEnumerator0)
    Socket Reader=$System.Socket Reader
    Socket Sampling Mode=::NewEnumerator3
    VelocitySpeedLimit=false
   )
   &ScaleColor(
    Scale Alpha=$Particles.Jitter
    Scale RGB=&VectorFromFloat(Value=$Particles.Jitter)
    ScaleA=true
    ScaleRGB=true
   )
   &FN_Sprite_Attributes(
    Camera Offset:NiagaraFloat=&FN_FrontEndFloatModifier(Gameplay=50, Front End=10)
    Color=&MultiplyLinearColor(
     Linear Color=&LerpLinearColors(StartColor=<rgba 1 0.3844 0.2925 1>, EndColor=<rgba 1.25 0.2401 0 1>, LerpFactor=$User.ForwardDot)
     Scale Factor=&MakeVector4FromVectorAndFloat(
      "Vector (RGB)"=<1 1 1>
      "Float (Alpha)"=&Multiply_Float(
       A=&Lerp_Float(
        A=&MakeCustomFloatFromBool(Bool=$User.bIsFrontEndPreview, True Float=1, False Float=0.25)
        B=0.35
        Alpha=$User.ForwardDot
       )
       B=$Particles.Jitter
      )
     )
    )
    Create Camera Offset Attribute=true
    Create Color Attribute=true
    Create Position Attribute=true
    Create SpriteSize Attribute=true
    Position:NiagaraPosition=&AddVectorToPosition(Position=$Output.SocketLocation.SampledPosition, Vector=<0 -5 0>)
    SpriteSize=&Multiply_Vector2D_ByFloat(Float=&Lerp_Float(A=0.65, B=1, Alpha=$User.ForwardDot), Vector2D=<120 80>)
    Write to Module Namespace=true
   )
   &FN_Light_Attributes(
    Color=&Multiply_LinearColorByFloat(
     Linear Color=&LerpLinearColors(StartColor=<rgba 1 0.4462 0.33 1>, EndColor=<rgba 1.35 0.519 0.3948 1>, LerpFactor=$User.ForwardDot)
     Float=1
    )
    Create Color Attribute=true
    Create Exponent Attribute=true
    Create Position Attribute=true
    Create Radius Attribute=true
    Exponent=3
    Position=&AddVectorToPosition(
     Position=$Output.SocketLocation.SampledPosition
     Vector=&SelectVectorFromArray(Direct Array Index=$Particles.SpawnGroup, Vector Selection Array=DataInterface() {propertyValues {FloatData [<5 20 0>, <-5 20 0>]}})
    )
    Radius=&Lerp_Float(A=40, B=60, Alpha=$User.ForwardDot)
    Scale Color=&Multiply_Float(
     A=$Particles.Jitter
     B=&MakeCustomFloatFromBool(Bool=$User.bIsFrontEndPreview, True Float=35, False Float=70)
    )
   )
  ]
  rendererClasses [&NiagaraSpriteRendererProperties, &NiagaraLightRendererProperties]
  renderers [
   {rendererClass=&NiagaraSpriteRendererProperties}
   {rendererClass=&NiagaraLightRendererProperties, rendererIndex=1}
  ]
 }
}
summary {
 systemName=NS_Glider_NoiseClueLurch_Fireworks_Idle
 emitters [
  {emitterName=Sparks, rendererClasses [&NiagaraSpriteRendererProperties]}
  {emitterName=Flame, rendererClasses [&NiagaraMeshRendererProperties]}
  {emitterName=Sound_Rings, rendererClasses [&NiagaraMeshRendererProperties]}
  {emitterName=Thursters, rendererClasses [&NiagaraSpriteRendererProperties, &NiagaraLightRendererProperties]}
  {emitterName=Fire, rendererClasses [&NiagaraMeshRendererProperties]}
  {
   emitterName=Light
   rendererClasses [
    &NiagaraLightRendererProperties
    &NiagaraComponentRendererProperties
    &NiagaraSpriteRendererProperties
   ]
  }
  {emitterName=Lamp_Lights, rendererClasses [&NiagaraMeshRendererProperties]}
 ]
 userVariables [
  User.bIsFrontEnd {type=NiagaraBool}
  User.bIsFrontEndPreview {type=NiagaraBool}
  User.ForwardDot {type=NiagaraFloat, defaultValue=1, description=""}
 ]
}
