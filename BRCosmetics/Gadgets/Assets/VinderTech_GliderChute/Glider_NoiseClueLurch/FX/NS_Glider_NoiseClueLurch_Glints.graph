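# Niagara glint effect for the NoiseClueLurch glider cosmetic, flashing short-lived warm orange-gold flares on the glider.
# One sprite emitter fires two small bursts, using spawn groups to give them different sizes.
# Particles sample skeletal mesh locations with random offsets during update, while curves and lifetime lerps drive center, corona, flicker and halo parameters for four sprite renderers.
# Loop delay and lifetime switch between gameplay and front-end values.
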NiagaraSystem /BRCosmetics/Gadgets/Assets/VinderTech_GliderChute/Glider_NoiseClueLurch/FX/NS_Glider_NoiseClueLurch_Glints
dependencies {
 usedDataInterfaces [
  &NiagaraDataInterfaceColorCurve
  &NiagaraDataInterfaceCurve
  &NiagaraDataInterfaceVector2DCurve
  &NiagaraDataInterfaceArrayFloat2
  &NiagaraDataInterfaceSkeletalMesh
 ]
 usedDynamicInputs []
 usedModules [
  &self:SystemSpawnScript.SystemScriptSource.SystemScriptGraph.NiagaraNodeAssignment_0.SetVariables_#1
  &SystemState
 ]
}
emitters {Flare_Glints {
  emitterUpdateScript [
   &EmitterState(
    Life Cycle Mode=::NewEnumerator1
    Loop Behavior=::NewEnumerator1
    Loop Delay=&FN_FrontEndFloatModifier(Gameplay=0.75, Front End=1)
    Loop Duration Mode=::NewEnumerator1
    UseLoopDelay=true
   )
   &SpawnBurst_Instantaneous(Spawn Time=0.15, Use Spawn Probability=false, UseLoopCountLimit=true)
   &SpawnBurst_Instantaneous(
    Spawn Count=&RandomRangeInt(Minimum=2, Maximum=3)
    Spawn Group=1
    Use Spawn Probability=false
    UseLoopCountLimit=true
   ) {module=SpawnBurst_Instantaneous001}
  ]
  particleSpawnScript [
   &InitializeParticle(
    AdjustAlpha=false
    AdjustHue=true
    AdjustSaturation=false
    AdjustValue=false
    Color=&LerpLinearColors(
     StartColor=<rgba 1 0.546 0 1>
     EndColor=&ColorFromCurve(
      ColorCurve=DataInterface() {propertyValues {
        AlphaCurve {keys [<at 0 1 RCIM_Cubic>]}
        BlueCurve {keys [<at 0.3001 0.1369 RCIM_Cubic>, {arriveTangent=1.5219, interpMode=RCIM_Cubic, leaveTangent=1.5219, time=0.4989, value=0}, <at 0.6875 0.7265>]}
        GreenCurve {keys [
          <at 0.3001 0.6609 RCIM_Cubic>
          {arriveTangent=2.2731, interpMode=RCIM_Cubic, leaveTangent=2.2731, time=0.4989, value=0.535}
          <at 0.6875 1.5416>
         ]}
        RedCurve {keys [<at 0.3001 1 RCIM_Cubic>, {arriveTangent=2.581, interpMode=RCIM_Cubic, leaveTangent=2.581, time=0.4989, value=1}, <at 0.6875 2>]}
       }}
      CurveIndex=&RandomRangeFloat(Minimum=0, Maximum=1)
     )
     LerpFactor=0
    )
    Color Mode=::NewEnumerator3
    Hue Shift Range:Vector2f=<-0.05 0.05>
    Lifetime Max=&FN_FrontEndFloatModifier(Gameplay=0.55, Front End=0.85)
    Lifetime Min=0.4
    Lifetime Mode=::NewEnumerator1
    Sprite Rotation Angle=0.125
    Sprite Rotation Mode=::NewEnumerator1
    Sprite Size=&MaskVector2dBySpawnGroup(
     Masked Vector2D=&Multiply_Vector2D_ByFloat(Float=&RandomRangeFloat(Minimum=1.25, Maximum=1.5), Vector2D=<45 45>)
     Passthrough Vector2D=&Multiply_Vector2D_ByFloat(Float=&RandomRangeFloat(Minimum=0.45, Maximum=0.55), Vector2D=<45 45>)
     Spawn Group Mask=0
    )
    Sprite Size Mode=::NewEnumerator4
    UsePositionOffset=false
    Write Material Random=true
   )
   &CameraOffset(Camera Offset Amount=25) {module=CameraOffset001}
   /SetVariables() {isSetParameters=true}
  ]
  particleUpdateScript [
   &ParticleState()
   &ScaleSpriteSize(
    Uniform Curve Scale=2
    Uniform Curve Sprite Scale=DataInterface() {propertyValues {Curve {
       defaultValue=0
       keys [
        {interpMode=RCIM_Cubic, tangentMode=RCTM_User, time=0, value=0}
        <at 0.25 1>
        {arriveTangent=-1.3333, interpMode=RCIM_Cubic, leaveTangent=-1.3333, tangentMode=RCTM_Break, time=0.75, value=1}
        {interpMode=RCIM_Cubic, tangentMode=RCTM_Break, time=1, value=0}
       ]
      }}}
   )
   &ScaleColor(
    Scale Alpha=&FloatFromCurve(
     FloatCurve=DataInterface() {propertyValues {Curve {
        defaultValue=1
        keys [
         {interpMode=RCIM_Cubic, tangentMode=RCTM_Break, time=0, value=0}
         <at 0.45 1 RCIM_Cubic>
         {interpMode=RCIM_Cubic, tangentMode=RCTM_Break, time=1, value=0}
        ]
       }}}
     Scale Curve=10
    )
    ScaleA=true
    ScaleRGB=true
   )
   /SetVariables(
    Particles.CenterFlareColor:LinearColor=&MultiplyLinearColor(
     Linear Color=$Particles.Color
     Scale Factor=&MakeVector4_FromLinearColor(LinearColor=&LerpLinearColors(StartColor=<rgba 5 3.7061 0 1>, EndColor=<rgba 5 4.0881 1.9271 1>, LerpFactor=$Particles.Lifetime))
    )
    Particles.CenterFlareSize:Vector2f=&Multiply_Vector2(A=$Particles.SpriteSize, B=&Vector2DFromFloat(Value=0.65))
    Particles.CoronaFlareColor:LinearColor=&MultiplyLinearColor(
     Linear Color=$Particles.Color
     Scale Factor=&MakeVector4_FromLinearColor(LinearColor=&LerpLinearColors(StartColor=<rgba 0.15 0.1065 0 1>, EndColor=<rgba 0.0729 0.0642 0 1>, LerpFactor=$Particles.Lifetime))
    )
    Particles.CoronaFlareSize:Vector2f=&Multiply_Vector2(A=$Particles.Initial.SpriteSize, B=&Vector2DFromFloat(Value=2))
    Particles.FlickerFlareColor:LinearColor=&MultiplyLinearColor(
     Linear Color=$Particles.Color
     Scale Factor=&MakeVector4_FromLinearColor(LinearColor=&LerpLinearColors(StartColor=<rgba 1 0.5411 0 1>, EndColor=<rgba 0.0337 0.0417 0.0337 50>, LerpFactor=$Particles.Lifetime))
    )
   ) {isSetParameters=true}
   /SetVariables(Particles.HaloSize:Vector2f=&Vector2DFromCurve(
     Vector2Curve=DataInterface() {propertyValues {
       XCurve {keys [{interpMode=RCIM_Cubic, tangentMode=RCTM_Break, time=0, value=0}, {arriveTangent=2.9125, tangentMode=RCTM_User, time=1, value=1}]}
       YCurve {keys [{interpMode=RCIM_Cubic, tangentMode=RCTM_Break, time=0, value=0}, {arriveTangent=2.9125, tangentMode=RCTM_User, time=1, value=1}]}
      }}
     ScaleCurve=&SelectVector2DFromArray(
      Array Sampling Mode=::NewEnumerator0
      Direct Array Index=?
      Vector 2D Selection Array=DataInterface() {propertyValues {FloatData [<65 65>, <75 75>]}}
      Random Seed=0
     )
    )) {isSetParameters=true}
   &SkeletalMeshLocation(
    Bone Sampling Mode=::NewEnumerator4
    MaskSpawnGroup=false
    Sampled Position Offset=&RandomRangeVector(Minimum=<-15 0 -10>, Maximum=<15 0 20>)
    Skeletal Mesh=$System.Skeletal Mesh Reader
   )
  ]
  rendererClasses [&NiagaraSpriteRendererProperties]
  renderers [
   {rendererClass=&NiagaraSpriteRendererProperties}
   {rendererClass=&NiagaraSpriteRendererProperties, rendererIndex=1}
   {rendererClass=&NiagaraSpriteRendererProperties, rendererIndex=2}
   {rendererClass=&NiagaraSpriteRendererProperties, rendererIndex=3}
  ]
 }}
summary {
 systemName=NS_Glider_NoiseClueLurch_Glints
 emitters [{emitterName=Flare_Glints, rendererClasses [&NiagaraSpriteRendererProperties]}]
 userVariables [
  User.bIsFrontEnd {type=NiagaraBool}
  User.bIsFrontEndPreview {type=NiagaraBool}
 ]
}
